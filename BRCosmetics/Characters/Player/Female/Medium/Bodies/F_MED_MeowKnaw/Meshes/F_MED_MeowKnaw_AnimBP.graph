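# Post-copy AnimBP for the MeowKnaw female body, driving cosmetic bone fixes and tail visibility.
# AnimGraph copies the attached parent's pose, runs it through the MeowKnaw Control Rig, then additively offsets the left glute deform bone while crouching.
# On customization load, it checks backbling type and tail presence, caches the MiscOrTail mesh, and hides the tail when a cape or long backbling is worn.
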AnimBlueprint /BRCosmetics/Characters/Player/Female/Medium/Bodies/F_MED_MeowKnaw/Meshes/F_MED_MeowKnaw_AnimBP
target_skeleton=&/F_MED_MeowKnaw_Skeleton
graphs [
 AnimGraph {nodes [
   n1 = Control Rig(Source=Copy Pose From Mesh() {bUseAttachedParent=true}) {
    alphaBoolBlend {blendOption=Linear}
    controlRigAssetReference {blueprintRigClass=&/F_MED_MeowKnaw_CtrlRig.F_MED_MeowKnaw_CtrlRig_C}
   }
   n2 = "Transform (Modify) Bone - Bone: deform_glute_base_l"!(
    ComponentPose=Local To Component(LocalPose=@n1)
    bAlphaBoolEnabled:bool=Get bIsCrouching()
    Translation="-4,0,0"
    Rotation="0,0,0"
    Scale="1,1,1"
   ) -> (Pose) {
    alphaInputType=Bool
    translationMode=BMM_Additive
    translationSpace=BCS_BoneSpace
    alphaBoolBlend {blendOption=Linear}
    boneToModify {boneName=deform_glute_base_l}
    translation {x=-4}
   }
   Output Pose(Result=Component To Local(ComponentPose=@n2)) {layerGroup=None}
  ]}
 EventGraph {nodes [
   n1 = Event OnCharacterCustomizationLoaded()
   n2 = CheckForBackbling()
   > Set isWearingBackbling(isWearingBackbling=AND Boolean!(A=@n2, B=NOT Boolean(A=@n2.bIsBackblingACape), C:bool=NOT Boolean(A=@n2.bIsBackblingLong)) -> (ReturnValue))
   > Set Is Wearing Cape(Is Wearing Cape=OR Boolean(A=@n2.bIsBackblingACape, B=@n2.bIsBackblingLong))
   > Set TailMesh(TailMesh:Skeletal Mesh Component Object Reference=Get Array Item with EFort Custom Part Type(Array=@n1.CustomMeshParts, PartType=MiscOrTail)) -> (Output_Get:Skeletal Mesh Component Object Reference)
   > n3 = Sequence()
   n4 = CheckForBackbling()
   > Set isWearingCape(isWearingCape=OR Boolean(A=@n4.bIsBackblingACape, B=@n4.bIsBackblingLong))
   > Set characterHasTail(characterHasTail:bool=CheckForTail!(self:Custom Character Part Anim Instance Object Reference) -> (bHasTail:bool)) -> (Output_Get:bool)
   Branch(execute=@n3.then_1, Condition=Get characterHasTail() -> (characterHasTail:bool))
   > n5 = Branch(Condition=Get isWearingCape() -> (isWearingCape:bool))
   n6 = Get TailMesh() -> (TailMesh:Skeletal Mesh Component Object Reference)
   > SetVisibility!(execute, self:Scene Component Object Reference=@n6, bNewVisibility:bool=false, bPropagateToChildren:bool=false) -> (then)
   SetVisibility(execute=@n5.else, self=@n6, bNewVisibility=true, bPropagateToChildren=false)
   Event BlueprintUpdateAnimation()
  ]}
]
variables [
 characterHasTail {type=boolean}
 Is Wearing Cape {type=boolean}
 isWearingBackbling {type=boolean}
 isWearingCape {type=boolean}
 TailMesh {type=SkeletalMeshComponent}
]
